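# Sound cue for the searching rustle of the ButterCake fruit plant, layering randomized rustle sounds.
# A mixer combines a random pick of three rustles through a default modulator with a briefly delayed, lower-pitched pick of two other rustles; its first input is empty.
# It borrows weapon-drop attenuation and the loot-collect sound class.
# Despite the Loop name, the graph contains no looping node.
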SoundCue /ButterCake/ButterCakeFruit/Audio/ButterCakeFruitPlant_Searching_Loop_Cue
root=Mixer_0
nodes [
 {id=Delay_0, children [Modulator_1], properties {delayMax=0.45, delayMin=0.35}}
 {id=Mixer_0, children [null,Modulator_0,Delay_0], properties {inputVolume [1.5,1.5,1]}}
 {id=Modulator_0, children [Random_1]}
 {id=Modulator_1, children [Random_2], properties {pitchMax=0.85, pitchMin=0.75}}
 {id=Random_1, children [WavePlayer_2,WavePlayer_5,WavePlayer_6], properties {bShouldExcludeFromBranchCulling=true, weights [1,1,1]}}
 {id=Random_2, children [WavePlayer_12, WavePlayer_13], properties {bShouldExcludeFromBranchCulling=true, weights [1,1]}}
 {id=WavePlayer_12, properties {soundWaveAssetPtr=&/Buttercake_Fruit_Plant_Search_Rustle_04}}
 {id=WavePlayer_13, properties {soundWaveAssetPtr=&/Buttercake_Fruit_Plant_Search_Rustle_05}}
 {id=WavePlayer_2, properties {soundWaveAssetPtr=&/Buttercake_Fruit_Plant_Search_Rustle_01}}
 {id=WavePlayer_5, properties {soundWaveAssetPtr=&/Buttercake_Fruit_Plant_Search_Rustle_02}}
 {id=WavePlayer_6, properties {soundWaveAssetPtr=&/Buttercake_Fruit_Plant_Search_Rustle_03}}
]
properties {
 attenuationSettings=&Fort_Weapon_Drop_Att
 maxDistance=3000
 soundClassObject=&Fort_Loot_Collect
 volumeMultiplier=0.5
}
