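# Sound cue for the ButterCake fruit drop, playing a one-shot variation of a fruit plant drop sound.
# A Random node picks among three equally weighted wave players and is excluded from branch culling.
# It uses the shared Fortnite weapon-drop attenuation and the loot-collect sound class, with volume trimmed slightly below full.
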SoundCue /ButterCake/ButterCakeFruit/Audio/ButterCakeFruit_Drop_Cue
root=Random_0
nodes [
 {id=Random_0, children [WavePlayer_1,WavePlayer_2,WavePlayer_4], properties {bShouldExcludeFromBranchCulling=true, weights [1,1,1]}}
 {id=WavePlayer_1, properties {soundWaveAssetPtr=&/Buttercake_Fruit_Plant_Drop_01}}
 {id=WavePlayer_2, properties {soundWaveAssetPtr=&/Buttercake_Fruit_Plant_Drop_02}}
 {id=WavePlayer_4, properties {soundWaveAssetPtr=&/Buttercake_Fruit_Plant_Drop_03}}
]
properties {
 attenuationSettings=&Fort_Weapon_Drop_Att
 maxDistance=3000
 soundClassObject=&Fort_Loot_Collect
 volumeMultiplier=0.95
}
