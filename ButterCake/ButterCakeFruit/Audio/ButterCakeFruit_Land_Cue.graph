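# Sound cue for the landing sound of a ButterCake fruit.
# A Random node picks one of three equally weighted plant-land wave variants and is set to be excluded from branch culling.
# The result passes through a Modulator at the root for pitch and volume variation.
# The cue plays at reduced volume, using the weapon-drop attenuation and the loot-collect sound class.
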SoundCue /ButterCake/ButterCakeFruit/Audio/ButterCakeFruit_Land_Cue
root=Modulator_2
nodes [
 {id=Modulator_2, children [Random_1]}
 {id=Random_1, children [WavePlayer_0,WavePlayer_3,WavePlayer_5], properties {bShouldExcludeFromBranchCulling=true, weights [1,1,1]}}
 {id=WavePlayer_0, properties {soundWaveAssetPtr=&/Buttercake_Fruit_Plant_Land_01}}
 {id=WavePlayer_3, properties {soundWaveAssetPtr=&/Buttercake_Fruit_Plant_Land_02}}
 {id=WavePlayer_5, properties {soundWaveAssetPtr=&/Buttercake_Fruit_Plant_Land_03}}
]
properties {
 attenuationSettings=&Fort_Weapon_Drop_Att
 maxDistance=3000
 soundClassObject=&Fort_Loot_Collect
 volumeMultiplier=0.3
}
