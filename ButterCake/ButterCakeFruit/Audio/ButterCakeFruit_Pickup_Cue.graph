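# ButterCakeFruit_Pickup_Cue is the sound cue for picking up the ButterCake fruit plant.
# A Random node picks one of three pickup wave variants with equal weight and is excluded from branch culling.
# The cue uses the loot-collect sound class and weapon-drop attenuation, with slightly lowered volume.
# Overridden concurrency allows only one instance per owner, so repeated pickups do not stack.
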SoundCue /ButterCake/ButterCakeFruit/Audio/ButterCakeFruit_Pickup_Cue
root=Random_0
nodes [
 {id=Random_0, children [WavePlayer_0,WavePlayer_3,WavePlayer_4], properties {bShouldExcludeFromBranchCulling=true, weights [1,1,1]}}
 {id=WavePlayer_0, properties {soundWaveAssetPtr=&/Buttercake_Fruit_Plant_Pickup_01}}
 {id=WavePlayer_3, properties {soundWaveAssetPtr=&/Buttercake_Fruit_Plant_Pickup_02}}
 {id=WavePlayer_4, properties {soundWaveAssetPtr=&/Buttercake_Fruit_Plant_Pickup_03}}
]
properties {
 attenuationSettings=&Fort_Weapon_Drop_Att
 bOverrideConcurrency=true
 maxDistance=3000
 soundClassObject=&Fort_Loot_Collect
 volumeMultiplier=0.85
 concurrencyOverrides {bLimitToOwner=true, maxCount=1}
}
