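# Sound cue named for the spawn sound of the ButterCake fruit projectile, currently empty with no nodes or root, so it references no audio to play.
# It retains only playback settings: reduced volume, the loot-collect sound class, and the weapon-drop attenuation asset with a limited maximum distance.
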SoundCue /ButterCake/ButterCakeFruit/Audio/Prj_ButterCakeFruit_Spawn_Cue
root=null
nodes []
properties {
 attenuationSettings=&Fort_Weapon_Drop_Att
 maxDistance=3200
 soundClassObject=&Fort_Loot_Collect
 volumeMultiplier=0.6
}
