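; Harvestable berry plant (ButterCakeFruit) that yields loot-table items when interacted with or destroyed.
; On the server, a game-data row sets hasLoot once playlist data is ready; its OnRep updates client berry visibility and plays a bounce animation.
; Interaction grants a rolled item with locally heard pickup audio if inventory has room, otherwise spawns floor loot, as does death.
; BeginPlay also forces custom mesh collision responses.
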(event EventBeginPlay
  (CallFunction|ForceSetMeshCollision)
  ;[1 1 1] Enable/Disable Interaction and Berry Visual Based on row in ButterCakeFruit_GameData
  (if (Networking|IsServer)
    (Class|FortAsyncActionCurrentPlaylistReady|PlaylistDataReadyAsync
      (:OnReady
        (Variables|Default|SethasLoot (Class|AthenaFunctionLibrary|RowtoBool (Variables|Default|GetRowBerriesEnabled)))))))

(event Interaction|EventBlueprintOnInteract
  (params (InteractingPawn :type Fort%20Pawn%20Object%20Reference) (InteractionBeingAttempted :type EInteractionBeingAttempted%20Enum :default "FirstInteraction"))
  (|CallParentFunction InteractingPawn InteractionBeingAttempted)
  (if (Variables|Default|GethasLoot)
    ;[1 1 1] Set the last pawn to interact and look for open inventory
    (bind _output_get (Variables|Default|SetLastInteractingPawn InteractingPawn))
    (bind _returnvalue (CallFunction|HasOpenInventorySlot _output_get))
    (if _returnvalue
      ;[1 1 1] If open give them the item
      (CallFunction|GiveItemtoInventory)
      (CallFunction|OnInteractAudio)
      (else
        ;[1 1 1] If not open, put it on the floor
        (CallFunction|SpawnLoot)))))

(event Custom|GiveItemToInventory
  ;[1 1 1] Get the last pawn to interact and cast to them
  (Development|PrintString "GivingItem")
  (bind _asfort_player_controller (Utilities|Casting|CastToFortPlayerController (Pawn|GetController (Variables|Default|GetLastInteractingPawn))))
  ;[1 1 1] Get the item we're giving and cast to it
  (bind (_item_definition _count) (CallFunction|GetItemtoGrant))
  (bind _asfort_world_item_definition (Utilities|Casting|CastToFortWorldItemDefinition _item_definition))
  (Variables|Default|SethasLoot false)
  ;[1 1 1] Give the chosen item to the last to interact
  (Inventory|GiveItemtoInventoryOwner :InventoryOwner _asfort_player_controller :ItemDefinition _asfort_world_item_definition :NumberToGive _count :bNotifyPlayer true :ItemLevel -1 :PickupInstigatorHandle -1 :bUseItemPickupAnalyticEvent false :WeaponAmmoOverride -1))

(event Custom|SpawnLoot
  ;[1 1 1] Spawn Loot
  (Variables|Default|SethasLoot false)
  (Class|AthenaFunctionLibrary|SpawnLootfromTable (Variables|Default|GetLootTierGroupName) 1 -1 (Transformation|GetActorLocation) "0, 0,10.000000" 0 true))

(event EventOnDeathServer
  (params (Damage :type Float%20%28single-precision%29) (DamageTags :type Gameplay%20Tag%20Container%20Structure%20%28by%20ref%29) (Momentum :type Vector :default "0, 0, 0") (HitInfo :type Hit%20Result%20Structure%20%28by%20ref%29) (InstigatedBy :type Controller%20Object%20Reference) (DamageCauser :type Actor%20Object%20Reference) (EffectContext :type Gameplay%20Effect%20Context%20Handle%20Structure))
  (if (Variables|Default|GethasLoot)
    (CallFunction|SpawnLoot)))

(event EventOnDeathPlayEffects
  (params (Damage :type Float%20%28single-precision%29) (DamageTags :type Gameplay%20Tag%20Container%20Structure%20%28by%20ref%29) (Momentum :type Vector :default "0, 0, 0") (HitInfo :type Hit%20Result%20Structure%20%28by%20ref%29) (InstigatedBy :type Fort%20Pawn%20Object%20Reference) (DamageCauser :type Actor%20Object%20Reference) (EffectContext :type Gameplay%20Effect%20Context%20Handle%20Structure))
  (Development|SetHiddeninGame (Variables|Debug|GetStaticMeshComponent) true true))

(event Custom|OnInteractAudio
  ;[1 1 1 0.3] Get Pick Sound from Item Def and then set it to play it 2D
  (bind _wid (Variables|Default|GetWID)
    (:then
      (bind _asfort_weapon_item_definition (Utilities|Casting|CastToFortWeaponItemDefinition _wid))
      (bind _itemdata (Class|ItemDefinitionBase|GetItemData _asfort_weapon_item_definition)
        (:Found
          (if (Utilities|IsValidSoftObjectReference (.PickupSound (Utilities|Struct|BreakFortnitePickupItemData _itemdata)))
            (Utilities|AsyncLoadAsset (.PickupSound (Utilities|Struct|BreakFortnitePickupItemData _itemdata))
              (:Completed
                (if (Pawn|IsLocallyViewed (Variables|Default|GetLastInteractingPawn))
                  (Audio|PlaySound2D (Utilities|ResolveSoftReference (.PickupSound (Utilities|Struct|BreakFortnitePickupItemData _itemdata)))))))))))
    (:else)))

(fn BlueprintCanInteract
  (params (InteractingPawn :type Fort%20Pawn%20Object%20Reference) (InteractionBeingAttempted :type EInteractionBeingAttempted%20Enum :default "FirstInteraction") (InteractionType :type TInteractionType%20Enum :default "IT_NoInteraction"))
  (if (Variables|Default|GethasLoot)
    (return true)))

(fn BlueprintGetInteractionString
  (params (InteractingPawn :type Fort%20Pawn%20Object%20Reference) (InteractionBeingAttempted :type EInteractionBeingAttempted%20Enum :default "FirstInteraction"))
  (return (Variables|Default|GetInteractionLabel)))

(fn ForceSetMeshCollision
  (bind _staticmeshcomponent (Variables|Debug|GetStaticMeshComponent))
  (Collision|SetCollisionResponsetoAllChannels _staticmeshcomponent "ECR_Ignore")
  (for _array_element (Utilities|Array|MakeArray "ECC_GameTraceChannel1" "ECC_GameTraceChannel2" "ECC_GameTraceChannel14" "ECC_GameTraceChannel16")
    (Collision|SetCollisionResponsetoChannel _staticmeshcomponent _array_element "ECR_Overlap"))
  (for _array_element_2 (Utilities|Array|MakeArray "ECC_Visibility" "ECC_GameTraceChannel3" "ECC_Destructible" "ECC_GameTraceChannel18" "ECC_Vehicle")
    (Collision|SetCollisionResponsetoChannel (Variables|Debug|GetStaticMeshComponent) _array_element_2 "ECR_Block")))

(fn GetItemToGrant
  (bind (_outloottodrop _returnvalue) (Loot|PickLootDrops (Variables|Default|GetLootTierGroupName) 0 0 "(GameplayTags=)"))
  (if _returnvalue
    (return (Class|FortItemFunctionLibrary|BreakFortItemEntry (Utilities|Array|Get(aref) _outloottodrop)) (.Count (Class|FortItemFunctionLibrary|BreakFortItemEntry (Utilities|Array|Get(aref) _outloottodrop))))))

(fn HasOpenInventorySlot
  (params (Target :type Pawn%20Object%20Reference))
  (bind _asfort_player_controller (Utilities|Casting|CastToFortPlayerController (Pawn|GetController Target))
    (:then
      (bind _returnvalue (IndexSelectionModel|GetNumItems(Message) _asfort_player_controller))
      (return (< _returnvalue 5)))
    (:CastFailed
      (return false))))

(fn OnRep_hasLoot
  (bind _returnvalue (Transformation|GetActorLocation))
  (bind _returnvalue_2 (+ _returnvalue "0,0,90.000000"))
  (if (not (Networking|IsServer))
    (Rendering|SetVisibility (Variables|Default|GetHierarchicalInstancedStaticMesh) (Variables|Default|GethasLoot) false)
    (Utilities|IsValid (Variables|Default|GetLastInteractingPawn)
      (:Is%20Valid
        (CallFunction|StartBounceAnimation (Utilities|Time|GetTimeSeconds) 1.2e+02 (Math|Vector|Normalize (- _returnvalue (Transformation|GetActorLocation (Variables|Default|GetLastInteractingPawn)))) _returnvalue_2 "Interact" true))
      (:Is%20Not%20Valid
        (CallFunction|StartBounceAnimation (Utilities|Time|GetTimeSeconds) 1.2e+02 (Math|Vector|MakeVector 0.0 0.0 1.0) _returnvalue_2 "Interact" true)))))

(fn ConstructionScript
  (|CallParentFunction))
