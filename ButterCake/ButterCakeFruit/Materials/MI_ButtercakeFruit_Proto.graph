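# Prototype material for the ButterCake fruit, a Default Lit placeholder surface despite the MI_ instance prefix in its name.
# The graph holds only an orange constant color and a single scalar constant, with no textures or parameters.
# No links from these constants to material outputs are shown.
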Material /ButterCake/ButterCakeFruit/Materials/MI_ButtercakeFruit_Proto
graph {
 properties {pixelDepthOffsetMode=PDOM_Legacy, usedShadingModels=Default Lit}
 nodes [
  Constant3Vector() {constant=<rgba 0.6875 0.2092 0 0>}
  Constant() {r=0.5}
 ]
}
