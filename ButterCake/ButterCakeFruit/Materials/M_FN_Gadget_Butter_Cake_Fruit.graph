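# M_FN_Gadget_Butter_Cake_Fruit is a Fortnite character-style master material for the Butter Cake fruit gadget, supporting multiple shading models.
# BaseFN PreFX, CharacterSkinFX, PostFX and CINE post functions build the base attributes.
# An optional WindNew switch adds panned swirl-noise offset and normal perturbation, masked by local height relative to the actor, with a red-green debug view.
# Mobile bypasses the wind, and the vertex-color mask is left unconnected.
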Material /ButterCake/ButterCakeFruit/Materials/M_FN_Gadget_Butter_Cake_Fruit
graph {
 properties {
  bAutomaticallySetUsageInEditor=false
  bUseMaterialAttributes=true
  bUsedWithClothing=true
  bUsedWithInstancedStaticMeshes=true
  bUsedWithMeshDeformer=true
  bUsedWithMorphTargets=true
  bUsedWithSkeletalMesh=true
  pixelDepthOffsetMode=PDOM_Legacy
  refractionMethod=RM_IndexOfRefraction
  shadingModel=MSM_FromMaterialExpression
  subsurfaceProfile=&SS_HeroSkin_02
  usedShadingModels=Default Lit | Subsurface | Clear Coat | Subsurface Profile | Cloth | Eye
 }
 nodes [
  n1 = MaterialFunctionCall() {materialFunction=&BaseFN_CharacterFunction_PreFX}
  n2 = MaterialFunctionCall(
   AO=@n1.AO
   Distance Blend=@n1.AO
   Fresnel=@n1.AO
   Gmap=@n1.AO
   Pre Skinned Local Position=@n1.AO
   Material=MaterialFunctionCall(
    PreskinnedPosition=@n1.Pre skinned Local Position
    Gmap=@n1.Pre skinned Local Position
    MaterialAttributes=@n1.Pre skinned Local Position
    AO=@n1.Pre skinned Local Position
   ) {materialFunction=&CharacterSkinFX}.Result
  ) {materialFunction=&BaseFN_CharacterFunction_PostFX}
  n3 = MaterialFunctionCall(In=@n2.Result) {materialFunction=&CINE_MF_CharPostFX}
  n4 = GetMaterialAttributes(None=@n3.Result) {attributeGetTypes [#1, #2]}
  n5 = WorldPosition() {worldPositionShaderOffset=WPT_ExcludeAllShaderOffsets}
  n6 = Transform(None=@n5.XYZ) {transformSourceType=TRANSFORMSOURCE_World, transformType=TRANSFORM_Local}
  n7 = Constant3Vector() {constant=<rgba 256 256 0 0>}
  n8 = Abs(None=Constant3Vector() {constant=<rgba 512 512 384 0>})
  n9 = ComponentMask(None=Divide(A=Add(A=@n6, B=@n7) {constB=256}, B=Multiply(A=@n8) {constB=-1})) {b=true, r=true}
  n10 = ScalarParameter() {defaultValue=0.5, group=WindNew, parameterName=WPO_WindNew_Scale, sortPriority=1}
  n11 = ScalarParameter() {defaultValue=0.05, group=WindNew, parameterName=WPO_WindNew_Speed_X, sortPriority=2}
  n12 = ScalarParameter() {defaultValue=-0.1, group=WindNew, parameterName=WPO_WindNew_Speed_Y, sortPriority=2}
  n13 = TextureSampleParameter2D(UVs=Panner(Coordinate=Multiply(A=@n9, B=@n10), Speed=AppendVector(A=@n11, B=@n12))) {
   group=WindNew
   parameterName=WPO_WindNew_Texture
   samplerSource=SSM_Wrap_WorldGroupSettings
   samplerType=SAMPLERTYPE_Normal
   texture=/Game/Global/Textures/Default/Noise/T_SwirlNoise_N
  }
  n14 = Multiply(
   A=Multiply(A=@n13.RGB, B=Constant3Vector() {constant=<rgba 1 1 0 0>})
   B=ScalarParameter() {defaultValue=0.5, group=WindNew, parameterName=WPO_WindNew_NormalsValue, sortPriority=4}
  )
  n15 = Subtract(A=ActorPositionWS(), B=@n5.XYZ)
  n16 = Divide(
   A=Add(
    A=Transform(None=@n15) {transformSourceType=TRANSFORMSOURCE_World, transformType=TRANSFORM_Local}
    B=ScalarParameter() {defaultValue=350, group=WindNew, parameterName=WPO_WindNew_Starting Point, sortPriority=5}
   )
   B=ScalarParameter() {defaultValue=155, group=WindNew, parameterName=WPO_WindNew_Fade, sortPriority=6}
  )
  n17 = ComponentMask(None=Max(A=Saturate(None=@n16), B=ScalarParameter() {group=WindNew, parameterName=WPO_Wind_Min})) {b=true}
  n18 = Constant() {r=1}
  n19 = AppendVector(
   A=ScalarParameter() {defaultValue=2, group=WindNew, parameterName=WPO_WindNew__X Val, sortPriority=3}
   B=ScalarParameter() {defaultValue=7, group=WindNew, parameterName=WPO_WindNew_Y Val, sortPriority=3}
  )
  n20 = Constant()
  n21 = Multiply(A=@n18, B=Multiply(A=@n13.RGB, B=AppendVector(A=@n19, B=@n20)))
  n22 = SetMaterialAttributes!(
   MaterialAttributes=@n4.MaterialAttributes
   Normal=LinearInterpolate(A=@n4.Normal, B=Add(A=@n4.Normal, B=@n14), Alpha=Multiply(A=@n17, B=@n18))
   World Position Offset=Add(A=@n4.World Position Offset, B=Multiply(A=@n17, B=@n21))
  ) {attributeSetTypes [#1, #2]}
  n23 = LinearInterpolate(A=Constant3Vector() {constant=<rgba 1 0 0 0>}, B=Constant3Vector() {constant=<rgba 0 1 0 0>}, Alpha=@n17)
  n24 = StaticSwitchParameter(
   True=StaticSwitchParameter(True=SetMaterialAttributes!(MaterialAttributes=@n22, Base Color=@n23) {attributeSetTypes [#3]}, False=@n22) {group=WindNew, parameterName=DebugWindMask, sortPriority=8}
   False=@n3.Result
  ) {group=WindNew, parameterName=WPO_WindNew, sortPriority=-1}
  ShadingPathSwitch(Default=@n24, Deferred=@n24, Forward=@n24, Mobile=@n3.Result)
  n25 = VertexColor()
  StaticComponentMaskParameter(None=AppendVector(A=@n25, B=@n25)) {defaultA=true, group=WindNew, parameterName=WPO_WindNew_VertexChannelMask, sortPriority=7}
 ]
}
