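# M_TrentProtoShader is a two-sided Default Lit material for the Butter Cake Fruit asset, which also serves as its preview mesh.
# It holds two texture parameters, TextureDiffuse for color and a grayscale TextureAO for ambient occlusion, defaulting to the cake's own textures.
# Two scalar constants, zero and one, accompany them, with no visible links from any node to material outputs.
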Material /ButterCake/ButterCakeFruit/Materials/M_TrentProtoShader
graph {
 properties {
  pixelDepthOffsetMode=PDOM_Legacy
  previewMesh=/ButterCake/ButterCakeFruit/Meshes/Butter_Cake_Fruit
  twoSided=true
  usedShadingModels=Default Lit
 }
 nodes [
  Constant()
  Constant() {r=1}
  TextureSampleParameter2D() {parameterName=TextureDiffuse, texture=/ButterCake/ButterCakeFruit/Textures/T_ButterCakeFruit_D}
  TextureSampleParameter2D() {parameterName=TextureAO, samplerType=SAMPLERTYPE_Grayscale, texture=/ButterCake/ButterCakeFruit/Textures/T_ButterCakeFruit_AO}
 ]
}
